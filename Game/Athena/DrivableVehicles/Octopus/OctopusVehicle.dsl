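; Blueprint for the drivable Octopus vehicle with boost and a grappling towhook, mostly handling cosmetic feedback.
; Enter and exit events toggle coil FX, audio and camera shake, and apply or remove weapon and cooldown effects.
; Towhook events drive a self-re-arming timer that stretches and wiggles the rope mesh, while timelines animate retraction and boost camera.
; Other events handle impacts, health-based damage states, water FX, destruction and pickaxe knockback.
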(event EventBeginPlay
  ;[1 1 1] BeginPlay
  (Networking|CanExecuteCosmeticEvents
    (:True
      (EventDispatchers|BindEventtoOnTowhookAttached (EventDispatchers|CreateEvent :!event Custom|TowhookAttached))
      (EventDispatchers|BindEventtoOnTowhookDetached (EventDispatchers|CreateEvent :!event Custom|TowhookDetached))
      (EventDispatchers|BindEventtoOnTowhookShot (EventDispatchers|CreateEvent :!event Custom|TowhookShot))
      (EventDispatchers|BindEventtoOnTowhookMissed (EventDispatchers|CreateEvent :!event Custom|TowhookMissed))
      (CallFunction|CacheAudioPointers (Variables|Default|GetAudioMovement) (Variables|Default|GetAudioWind1P) (Variables|Default|GetAudioTowCable))
      (CallFunction|CacheParticleFX (Variables|Default|GetP_Octopus_Coil_Idle_TopR) (Variables|Default|GetP_Octopus_Coil_Idle_TopL) (Variables|Default|GetP_Octopus_Coil_Idle_LowR) (Variables|Default|GetP_Octopus_Coil_Idle_LowL) (Variables|Default|GetP_Octopus_Boost) (Variables|Default|GetP_Octopus_DustFX))
      (EventDispatchers|BindEventtoOnTowhookAttached (delegate Custom|EventOnTowhookAttached))
      (Health|BindEventtoOnHealthChanged (EventDispatchers|CreateEvent :!event Handle%20On%20Health%20Changed)))))

(event Custom|EventOnTowhookAttached
  (Development|SetHiddeninGame (Variables|Default|GetReticle) true false))

(event Custom|StopCameraShake
  (bind _drivingplayercontroller (Variables|Default|GetDrivingPlayerController))
  ;[1 1 1] Stop Camera Shake
  (Utilities|IsValid (Variables|FortOctopusVehicle|GetCacheDriverCameraShake)
    (:Is%20Valid
      (Utilities|IsValid _drivingplayercontroller
        (:Is%20Valid
          (CameraShakes|StopCameraShake (Variables|/Script/Engine.PlayerController|GetPlayerCameraManager _drivingplayercontroller) (Variables|FortOctopusVehicle|GetCacheDriverCameraShake) false)
          (Variables|FortOctopusVehicle|SetCacheDriverCameraShake 0))))))

(event Custom|StartDriverCameraShake
  (bind _drivingplayercontroller (Variables|Default|GetDrivingPlayerController))
  ;[1 1 1] Start Camera Shake
  (if (Vehicle|ClientIsDriver)
    (if (not (Variables|Default|GetDebug_NoCameraShake))
      (Utilities|IsValid _drivingplayercontroller
        (:Is%20Valid
          (bind _returnvalue (CameraShakes|StartCameraShake :self (Variables|/Script/Engine.PlayerController|GetPlayerCameraManager _drivingplayercontroller) :ShakeClass "/Game/Athena/DrivableVehicles/Speed_CameraShake.Speed_CameraShake_C" :Scale 1.0 :PlaySpace "CameraLocal"))
          (Variables|FortOctopusVehicle|SetCacheDriverCameraShake (CameraShakes|ConvLegacyCameraShake _returnvalue)))))))

(event Collision|EventActorBeginOverlap
  (params (OtherActor :type Actor%20Object%20Reference)))

(event EventOnPawnEnterVehicle
  (params (PlayerPawn :type Fort%20Player%20Pawn%20Object%20Reference) (SeatIdx :type Integer))
  (bind _audiogetinout (Variables|Default|GetAudioGetInOut))
  (bind _idlefxarray (Variables|Default|GetIdleFXArray))
  ;[1 1 1] Player In and Out
  (CallFunction|AddCooldownCueEvent PlayerPawn)
  (Utilities|FlowControl|Sequence
    (:then_0
      (bind _asplayer_controller (Utilities|Casting|CastToPlayerController (Pawn|GetController PlayerPawn)))
      (Variables|Default|SetDrivingPlayerController _asplayer_controller)
      (CallFunction|StartDriverCameraShake))
    (:then_1
      (Components|Activation|Activate (Variables|Default|GetIdleFXArray) true)
      (Audio|Components|Audio|SetSound _audiogetinout (Variables|Sounds|GetSoundOnGetIn))
      (Audio|Components|Audio|Play _audiogetinout 0.0)
      (bind (_direction _intensity) (Timeline|Startup PlayFromStart)
        (:Update
          (Effects|Components|ParticleSystem|SetFloatParameter _idlefxarray "BoostAlpha" _intensity)
          (Effects|Components|ParticleSystem|SetFloatParameter _idlefxarray "Boost" _intensity))))
    (:then_2
      ;; EquipWeapon
      (if (Vehicle|IsOperatingTurret PlayerPawn)
        (bind _returnvalue (Environment|ApplyGlobalEnvironmentGameplayEffecttoActor PlayerPawn (Variables|Default|GetGE_GrantWeapon)))
        (Variables|Default|SetGE_EnterTurret _returnvalue)))))

(event Custom|CameraShake
  (params (Scale :type Float%20%28double-precision%29))
  (Game|Feedback|ClientStartCameraShake :self (Fortnite|GetFortPlayerControllerfromActor (Vehicle|GetDriver)) :Shake (Variables|Default|GetBounceCameraShake) :Scale (Math|Float|MapRangeClamped Scale (Variables|Default|GetBounceEffectsThreshold) (Variables|Default|GetInRangeMax) (Variables|Default|GetInMax) (Variables|Default|GetOutMax)) :PlaySpace "CameraLocal"))

(event EventOnHit
  (params (Location :type Vector%20%28by%20ref%29 :default "0, 0, 0") (Normal :type Vector%20%28by%20ref%29 :default "0, 0, 0") (Force :type Float%20%28single-precision%29) (HitActor :type Actor%20Object%20Reference))
  ;[1 1 1] Impacts
  (if (not (Networking|IsDedicatedServer))
    (Utilities|FlowControl|Sequence
      (:then_0
        (if (>= Force (Variables|Default|GetBounceThreshold))
          (Utilities|FlowControl|DoOnce false
            (:Completed
              (Effects|SpawnEmitteratLocation "/Game/Athena/DrivableVehicles/FX/Octopus/P_Octopus_Impact.P_Octopus_Impact" Location (Math|Rotator|MakeRotfromX Normal))
              (Utilities|Time|SetTimerbyEvent (EventDispatchers|CreateEvent :!event Custom|ImpactReset) (Variables|Default|GetImpactResetTimer))
              (if (Vehicle|ClientIsDriver)
                (CallFunction|CameraShake Force))))))
      (:then_1
        ;; Audio Collision
        (if (not (Class|FortVehicleInterface|IsBoosting))
          (Audio|OnCollision (Variables|Default|GetNormalCollisionAudio) Location (* Normal Force)))))))

(event Custom|ImpactReset
  ;[1 1 1] Impacts
  (Utilities|FlowControl|DoOnce false
    (:Completed
      (Effects|SpawnEmitteratLocation "/Game/Athena/DrivableVehicles/FX/Octopus/P_Octopus_Impact.P_Octopus_Impact" (stale EventOnHit Location) (Math|Rotator|MakeRotfromX (stale EventOnHit Normal)))
      (Utilities|Time|SetTimerbyEvent (EventDispatchers|CreateEvent :!event Custom|ImpactReset) (Variables|Default|GetImpactResetTimer))
      (if (Vehicle|ClientIsDriver)
        (CallFunction|CameraShake (stale EventOnHit Force))))))

(event Custom|GroundHit
  (params (Scale :type Float%20%28double-precision%29))
  (Game|Feedback|ClientStartCameraShake :self (Fortnite|GetFortPlayerControllerfromActor (Vehicle|GetDriver)) :Shake "/Game/Athena/DrivableVehicles/Octopus/OctopusHitGroundShake.OctopusHitGroundShake_C" :Scale Scale :PlaySpace "CameraLocal"))

(event EventOnPawnExitVehicle
  (params (PlayerPawn :type Fort%20Player%20Pawn%20Object%20Reference) (ExitSocketName :type Name :default "None"))
  (bind _audiogetinout (Variables|Default|GetAudioGetInOut))
  (bind _idlefxarray (Variables|Default|GetIdleFXArray))
  ;[1 1 1] Player In and Out
  (Utilities|FlowControl|Sequence
    (:then_0
      (Components|Activation|Deactivate (Variables|Default|GetIdleFXArray))
      (Audio|Components|Audio|SetSound _audiogetinout (Variables|Sounds|GetSoundOnGetOut))
      (Audio|Components|Audio|Play _audiogetinout 0.0)
      (Timeline|Startup ReverseFromEnd))
    (:then_1
      (CallFunction|StopCameraShake)
      (Development|SetHiddeninGame (Variables|Default|GetReticle) true false))
    (:then_2
      (Utilities|IsValid PlayerPawn
        (:Is%20Valid
          ;; UnequipWeapon
          (GameplayEffects|RemoveActiveGameplayEffect (Ability|GetAbilitySystemComponent PlayerPawn) (Variables|Default|GetGE_EnterTurret) -1))))
    (:then_3
      (Utilities|IsValid PlayerPawn
        (:Is%20Valid
          (CallFunction|RemoveCooldownCueEvent PlayerPawn))))))

(event EventOnBoostBounce
  (params (Location :type Vector%20%28by%20ref%29 :default "0, 0, 0") (Normal :type Vector%20%28by%20ref%29 :default "0, 0, 0") (Force :type Float%20%28single-precision%29) (HitActor :type Actor%20Object%20Reference) (BounceCount :type Integer))
  ;[1 1 1] Impacts
  (Effects|SpawnEmitteratLocation "/Game/Athena/DrivableVehicles/FX/Octopus/P_Octopus_Impact_Boost.P_Octopus_Impact_Boost" Location (Math|Rotator|MakeRotfromX Normal))
  (if (Variables|Sounds|GetbCanPlayBoostBounceAudio)
    (Audio|PlaySoundatLocation "/Game/Athena/Sounds/Vehicles/Hamster/SC_Hamster_CollisionImpact_Boost_Cue.SC_Hamster_CollisionImpact_Boost_Cue" Location "0, 0, 0" 1.0 (Math|Float|MapRangeClamped (Math|Conversions|ToFloat(Integer) BounceCount) 0.0 12.0 0.9 2.0))))

(event EventOnBoostBegin
  (bind _idlefxarray (Variables|Default|GetIdleFXArray))
  ;[1 1 1] Boost FX
  (CallFunction|BoostCameraBegin)
  (Components|Activation|Activate (Variables|Default|GetP_Octopus_Boost) true)
  (Effects|Components|ParticleSystem|SetFloatParameter _idlefxarray "BoostAlpha" 1.0)
  (Effects|Components|ParticleSystem|SetFloatParameter _idlefxarray "Boost" 1.0)
  (Audio|Components|Audio|Play (Variables|Default|GetAudioBoost) 0.0)
  (Utilities|Time|SetTimerbyEvent (delegate Custom|Bounce%20Collision%20Audio%20Cooldown) 0.2)
  (Variables|Sounds|SetbCanPlayBoostBounceAudio false))

(event Custom|Bounce%20Collision%20Audio%20Cooldown
  ;[1 1 1] Boost FX
  (Variables|Sounds|SetbCanPlayBoostBounceAudio true))

(event EventOnBoostEnd
  (bind _idlefxarray (Variables|Default|GetIdleFXArray))
  ;[1 1 1] Boost FX
  (CallFunction|BoostCameraReverse)
  (Components|Activation|Deactivate (Variables|Default|GetP_Octopus_Boost))
  (Effects|Components|ParticleSystem|SetFloatParameter _idlefxarray "BoostAlpha" 0.0)
  (Effects|Components|ParticleSystem|SetFloatParameter _idlefxarray "Boost" 0.0))

(event Custom|BoostCameraBegin
  ;[1 1 1] Boost Effects
  (Variables|Default|SetbBoostCameraActive true)
  (Components|Timeline|SetPlayRate (Variables|OctopusVehicle|GetBoostTimeline) (Variables|Default|GetBoostTimelinePlayRate))
  (bind (_direction _rumble _newtrack_0) (Timeline|BoostTimeline Play)
    (:Update
      (Variables|Camera|SetCameraFOVOffset (Math|Float|Lerp 0.0 (Variables|Default|GetBoostFOVOffset) _newtrack_0))
      (Variables|Camera|SetCameraSpaceForwardDistanceOffset (Math|Float|Lerp 0.0 (Variables|Default|GetBoostCameraDistanceOffset) _newtrack_0))
      (Variables|Default|SetBoostRumbleValue _rumble)
      (switch Utilities|FlowControl|Switch|SwitchonETimelineDirection _direction
        (:Forward
          (Rendering|Material|SetScalarParameterValueonMaterials (Variables|Vehicle|GetSkeletalMesh) "BoostGlow" (Math|Float|Clamp(Float) (* _newtrack_0 1.0))))))
    (:Finished
      (switch Utilities|FlowControl|Switch|SwitchonETimelineDirection _direction
        (:Forward
          (CallFunction|BoostCameraReverse))
        (:Backward
          (Variables|Default|SetbBoostCameraActive false))))))

(event Custom|ReverseBoostTimeline
  ;[1 1 1] Boost Effects
  (Timeline|BoostTimeline Reverse))

(event Custom|BoostCameraReverse
  ;[1 1 1] Boost Effects
  (if (Variables|Default|GetbBoostCameraActive)
    (Components|Timeline|SetPlayRate (Variables|OctopusVehicle|GetBoostTimeline) (Variables|Default|GetBoostReversePlayRate))
    (CallFunction|ReverseBoostTimeline)))

(event EventOnUpdateTargetPoint
  (params (bAimAssistHitsTarget :type Boolean) (TargetPoint :type Vector%20%28by%20ref%29 :default "0, 0, 0"))
  (bind _reticle (Variables|Default|GetReticle))
  (Development|SetHiddeninGame _reticle (not bAimAssistHitsTarget) false)
  (Transformation|SetWorldLocation _reticle TargetPoint false false))

(event EventOnStopTick
  ;[1 1 1] Turn off audio on Stop Tick event
  (Audio|Components|Audio|FadeOut (Variables|Default|GetAudioMovement) 0.5 0.0 "Linear" :!class /Script/Engine.AudioComponent))

(event EventOnDamagePlayEffects
  (params (Damage :type Float%20%28single-precision%29) (DamageTags :type Gameplay%20Tag%20Container%20Structure%20%28by%20ref%29) (Momentum :type Vector :default "0, 0, 0") (HitInfo_bBlockingHit :type Boolean) (HitInfo_bInitialOverlap :type Boolean) (HitInfo_Time :type Float%20%28single-precision%29) (HitInfo_Distance :type Float%20%28single-precision%29) (HitInfo_Location :type Vector :default "0, 0, 0") (HitInfo_ImpactPoint :type Vector :default "0, 0, 0") (HitInfo_Normal :type Vector :default "0, 0, 0") (HitInfo_ImpactNormal :type Vector :default "0, 0, 0") (HitInfo_PhysMat :type Physical%20Material%20Object%20Reference) (HitInfo_HitActor :type Actor%20Object%20Reference) (HitInfo_HitComponent :type Primitive%20Component%20Object%20Reference) (HitInfo_HitBoneName :type Name :default "None") (HitInfo_BoneName :type Name :default "None") (HitInfo_HitItem :type Integer) (HitInfo_ElementIndex :type Integer) (HitInfo_FaceIndex :type Integer) (HitInfo_TraceStart :type Vector :default "0, 0, 0") (HitInfo_TraceEnd :type Vector :default "0, 0, 0") (InstigatedBy :type Fort%20Pawn%20Object%20Reference) (DamageCauser :type Actor%20Object%20Reference) (EffectContext :type Gameplay%20Effect%20Context%20Handle%20Structure))
  ;[1 1 1] Damage Health
  (GameplayCue|ExecuteGameplayCue%28Burst%29Local self (Variables|Default|GetImpactCue) (Ability|GameplayCue|MakeGameplayCueParameters :NormalizedMagnitude 0.0 :RawMagnitude 0.0 :Location HitInfo_Location :Normal (Math|Vector|Normalize (- (Transformation|GetActorLocation DamageCauser) HitInfo_Location)) :GameplayEffectLevel 1 :AbilityLevel 1 :bReplicateLocationWhenUsingMinimalRepProxy false)))

(event EventK2_ApplyCosmeticWrap
  (params (LoadedWrap :type Athena%20Item%20Wrap%20Definition%20Object%20Reference))
  (bind _output_materialinterface (Utilities|Struct|BreakSkeletalMaterial (Utilities|Array|Get(acopy) (Variables|/Script/Engine.SkeletalMesh|GetMaterials (Variables|/Script/Engine.SkinnedMeshComponent|GetSkeletalMesh (Variables|Vehicle|GetSkeletalMesh))))))
  (Class|FortVehicleInterface|ApplyMaterialandUseWrapifPossible _output_materialinterface))

(event Custom|TowhookShot
  (bind _returntimeline (Variables|OctopusVehicle|GetReturnTimeline))
  ;[1 1 1] Rope
  (Rendering|Material|SetScalarParameterValueonMaterials (Variables|Default|GetSM_Grapple_Rope) "ReturnWiggle" 0.0)
  (CallFunction|SetGrappleVisible true)
  (if (Components|Timeline|IsPlaying _returntimeline)
    (Components|Timeline|Stop _returntimeline)
    (Development|SetHiddeninGame (Variables|Default|GetCableReturnMeshComponent) true false))
  (CallFunction|TowhookLoop)
  (Variables|Default|SetbHasRopeRetracted false))

(event Custom|TowhookLoop
  (bind _cablestart (Variables|Default|GetCableStart))
  (bind _cableend (Variables|Default|GetCableEnd))
  (bind _returnvalue (- _cableend _cablestart))
  (bind _returnvalue_2 (Math|Vector|VectorLength _returnvalue))
  (bind _returnvalue_3 (/ _returnvalue_2 1e+02))
  (bind (_outcablestart _outcableend) (CallFunction|GetTowhookCableEndpoints))
  (bind _sm_grapple_rope (Variables|Default|GetSM_Grapple_Rope))
  ;[1 1 1] Rope
  (if (CallFunction|IsTowhookHolstered)
    (CallFunction|RetractCable)
    (else
      (Variables|Default|SetCableStart _outcablestart)
      (Variables|Default|SetCableEnd _outcableend)
      (Transformation|SetWorldLocationAndRotation (Variables|Default|GetSM_Grapple_Rope) _cablestart (Math|Rotator|MakeRotfromX _returnvalue) false false)
      (Rendering|Material|SetScalarParameterValueonMaterials _sm_grapple_rope "MaxLength" _returnvalue_2)
      (Transformation|SetRelativeScale3D _sm_grapple_rope (Math|Vector|MakeVector _returnvalue_3 1.0 1.0))
      (bind _returnvalue_4 (Utilities|Time|SetTimerbyEvent (EventDispatchers|CreateEvent :!event Custom|TowhookLoop) 0.0001))
      (Variables|Default|SetCableLoopTimer _returnvalue_4)
      (if (CallFunction|IsTowhookAttached)
        (bind _output_get (Variables|Default|SetWiggleAmount (Math|Interpolation|FInterpTo (Variables|Default|GetWiggleAmount) 0.0 (Utilities|Time|GetWorldDeltaSeconds) 2e+01)))
        (Rendering|Material|SetScalarParameterValueonMaterials (Variables|Default|GetSM_Grapple_Rope) "WiggleAmount" _output_get)
        (else
          (bind _output_get_2 (Variables|Default|SetWiggleAmount (/ _returnvalue_3 (Variables|Default|GetOutgoingWiggleFactor))))
          (Rendering|Material|SetScalarParameterValueonMaterials (Variables|Default|GetSM_Grapple_Rope) "WiggleAmount" _output_get_2))))))

(event Custom|SetGrappleVisible
  (params (Visible :type Boolean))
  (bind _sm_grapple_rope (Variables|Default|GetSM_Grapple_Rope))
  ;[1 1 1] Rope
  (if Visible
    (Rendering|SetVisibility _sm_grapple_rope true false)
    (else
      (Rendering|SetVisibility _sm_grapple_rope false false))))

(event EventOnDamaged
  (params (Damage :type Float%20%28single-precision%29) (DamageTags :type Gameplay%20Tag%20Container%20Structure%20%28by%20ref%29) (EffectContext :type Gameplay%20Effect%20Context%20Handle%20Structure%20%28by%20ref%29) (EventInstigator :type Controller%20Object%20Reference) (DamageCauser :type Actor%20Object%20Reference))
  ;[1 1 1] onDamage event on server
  (|CallParentFunction Damage DamageTags EffectContext EventInstigator DamageCauser)
  (if (GameplayTags|HasTag DamageTags "(TagName=\"Weapon.Melee.Impact.Pickaxe\")")
    (CallFunction|HitwithPickaxe (Math|Vector|GetForwardVector (Pawn|GetControlRotation EventInstigator :!class /Script/Engine.Controller)))))

(event Custom|TowhookAttached
  (bind (_outcablestart _outcableend) (CallFunction|GetTowhookCableEndpoints))
  (bind _returnvalue (Vehicle|GetAllPassengers))
  (bind _output (Utilities|Array|Get(acopy) _returnvalue))
  ;[1 1 1] Rope
  (Rendering|Material|SetScalarParameterValueonMaterials (Variables|Default|GetSM_Grapple_Rope) "ReturnWiggle" 0.0)
  (Utilities|IsValid _output
    (:Is%20Valid
      (if (Pawn|IsLocallyControlled _output)
        (Audio|PlaySound2D "/Game/Athena/Sounds/Vehicles/Hamster/SC_Hamster_Grappling_Attach_Cue.SC_Hamster_Grappling_Attach_Cue")
        (else
          _)))
    (:Is%20Not%20Valid
      _))
  (Audio|PlaySoundatLocation "/Game/Sounds/Fort_Weapons/GrapplingHook/Grappling_Impact_Cue.Grappling_Impact_Cue" _outcableend))

(event Custom|TowhookMissed
  (bind _cablereturnmeshcomponent (Variables|Default|GetCableReturnMeshComponent))
  ;[1 1 1] Rope
  (Utilities|Time|ClearandInvalidateTimerbyHandle (Variables|Default|GetCableLoopTimer))
  (CallFunction|RetractCable)
  (Audio|SpawnSoundAttached "/Game/Athena/Sounds/Vehicles/Hamster/SC_Hamster_Grappling_YankEnd_Cue.SC_Hamster_Grappling_YankEnd_Cue" (Variables|Vehicle|GetSkeletalMesh))
  (Utilities|FlowControl|Sequence
    (:then_0
      (Utilities|IsValid _cablereturnmeshcomponent
        (:Is%20Valid
          (Development|SetHiddeninGame _cablereturnmeshcomponent false false))
        (:Is%20Not%20Valid
          (bind _returnvalue (AddComponent|NODE_AddStaticMeshComponent-0 :RelativeTransform_Location "0, 0, 0" :RelativeTransform_Rotation "0, 0, 0" :RelativeTransform_Scale "1.500000,1.500000,1.500000"))
          (bind _output_get (Variables|Default|SetCableReturnMeshComponent _returnvalue))
          (Transformation|DetachFromComponent _output_get "KeepRelative" "KeepRelative" "KeepRelative" false))))
    (:then_1
      (Transformation|SetWorldRotation (Variables|Default|GetCableReturnMeshComponent) (Math|Rotator|MakeRotfromX (- (Variables|Default|GetCableEnd) (Variables|Default|GetCableStart))) false false)
      (Variables|Default|SetbTowhookMissed true))))

(event Custom|HitWithPickaxe
  (params (HitDirection :type Vector))
  ;[1 1 1] onDamage event on server
  (Audio|BroadcastAbilityImpactatLocation 0 self (* (+ HitDirection "0, 0,0.500000") (Variables|Default|GetPickaxeImpulseStrength)) 777.0)
  (Audio|PlaySoundatLocation "/Game/Athena/Sounds/Vehicles/Hamster/SC_Hamster_CollisionImpact_Cue.SC_Hamster_CollisionImpact_Cue" (Transformation|GetActorLocation)))

(event Custom|TowhookDetached
  (bind _returnvalue (Vehicle|GetAllPassengers))
  (bind _output (Utilities|Array|Get(acopy) _returnvalue))
  (bind _sm_grapple_rope (Variables|Default|GetSM_Grapple_Rope))
  ;[1 1 1] Rope
  (Rendering|Material|SetScalarParameterValueonMaterials _sm_grapple_rope "WiggleAmount" 15.0)
  (Rendering|Material|SetScalarParameterValueonMaterials _sm_grapple_rope "ReturnWiggle" 0.0)
  (Audio|SpawnSoundAttached "/Game/Athena/Sounds/Vehicles/Hamster/SC_Hamster_Grappling_YankEnd_Cue.SC_Hamster_Grappling_YankEnd_Cue" (Variables|Vehicle|GetSkeletalMesh))
  (Utilities|IsValid _output
    (:Is%20Valid
      (if (Pawn|IsLocallyControlled _output)
        (Audio|PlaySound2D "/Game/Athena/Sounds/Vehicles/Hamster/SC_Hamster_Wind_Whoosh_1P_Cue.SC_Hamster_Wind_Whoosh_1P_Cue" (Math|Float|MapRangeClamped (+ (select (Class|FortVehicleInterface|IsBoosting) 2e+03 0.0) (Math|Vector|VectorLength (Transformation|GetVelocity))) 1e+03 3e+03 0.0 1.0))))))

(event Custom|RetractCable
  (bind (_returnvalue_location _returnvalue_rotation) (Math|Transform|BreakTransform (Transformation|GetSocketTransform (Variables|Default|GetSM_Grapple_Rope) "end" "RTS_World")))
  (bind _relativescale3d_x (Math|Vector|BreakVector (Variables|/Script/Engine.SceneComponent|GetRelativeScale3D_X (Variables|Default|GetSM_Grapple_Rope))))
  (bind _sm_grapple_rope (Variables|Default|GetSM_Grapple_Rope))
  ;[1 1 1] Rope
  (bind _output_get (Variables|Default|SetRetractScaleStart _relativescale3d_x))
  (Components|Timeline|SetPlayRate (Variables|OctopusVehicle|GetReturnTimeline) (/ 1.0 (/ _output_get 1e+02)))
  (bind (_direction _newtrack_0) (Timeline|ReturnTimeline PlayFromStart)
    (:Update
      (Transformation|SetRelativeScale3D (Variables|Default|GetSM_Grapple_Rope) (Math|Vector|MakeVector (Math|Float|Lerp (Variables|Default|GetRetractScaleStart) 0.001 _newtrack_0) 1.0 1.0))
      (Rendering|Material|SetScalarParameterValueonMaterials _sm_grapple_rope "WiggleAmount" (Math|Float|Lerp 0.0 15.0 _newtrack_0))
      (Rendering|Material|SetScalarParameterValueonMaterials _sm_grapple_rope "ReturnWiggle" (Math|Float|Lerp 0.0 3.0 _newtrack_0))
      (Rendering|Material|SetScalarParameterValueonMaterials _sm_grapple_rope "MaxLength" (* (Variables|Default|GetRetractScaleStart) 1e+02))
      (if (Variables|Default|GetbTowhookMissed)
        (Transformation|SetWorldTransform (Variables|Default|GetCableReturnMeshComponent) (Math|Transform|MakeTransform (+ _returnvalue_location (* (Math|Vector|GetForwardVector _returnvalue_rotation) 2e+01)) _returnvalue_rotation "1.500000,1.500000,1.500000") false false)))
    (:Finished
      (Variables|Default|SetbHasRopeRetracted true)
      (CallFunction|SetGrappleVisible false)
      (if (Variables|Default|GetbTowhookMissed)
        (Variables|Default|SetbTowhookMissed false)
        (Development|SetHiddeninGame (Variables|Default|GetCableReturnMeshComponent) true false)))))

(event Vehicle|Water|EventOnEnteredWaterVolume
  (params (WaterSurfacePoint :type Vector%20%28by%20ref%29 :default "0, 0, 0"))
  (bind _skeletalmesh (Variables|Vehicle|GetSkeletalMesh))
  ;[1 1 1] EVENT -> On Entered Water Volume -- FX for water ineraction
  (Networking|CanExecuteCosmeticEvents
    (:True
      (if (> (Rendering|GetRenderingDetailMode) 0)
        (bind _returnvalue (Utilities|Time|SetTimerbyEvent (delegate Custom|CheckWater) 0.25 true))
        (bind _inwaterfx (Variables|Default|GetInWaterFX)
          (:then
            (Components|DestroyComponent _inwaterfx))
          (:else
            _))
        (bind _returnvalue_2 (Effects|SpawnEmitterAttached :EmitterTemplate "/Game/WaterAndWind/Particles/WaterInteraction/P_Vehicle_Octopus_In_Water_Athena.P_Vehicle_Octopus_In_Water_Athena" :AttachToComponent _skeletalmesh :AttachPointName "DriverSeat" :Location (0.0 0.0 0.0) :Scale "0.800000,0.800000,0.800000" :LocationType "KeepRelativeOffset" :bAutoDestroy true :PoolingMethod "None" :bAutoActivate true)) ; Looping water FX
        (bind _output_get (Variables|Default|SetInWaterFX _returnvalue_2))
        (Rendering|SetTranslucentSortPriority _output_get -1)
        (Effects|SpawnEmitterAttached :EmitterTemplate "/Game/WaterAndWind/Particles/WaterInteraction/P_WaterSplash_Athena_Vehicle.P_WaterSplash_Athena_Vehicle" :AttachToComponent _skeletalmesh :AttachPointName "DriverSeat" :Location (Math|Vector|MakeVector 0.0 0.0 (- 0.0 75.0)) :Scale "1.000000,1.000000,1.000000" :LocationType "KeepRelativeOffset" :bAutoDestroy true :PoolingMethod "None" :bAutoActivate true) ; Splash burst FX
        (bind _inwaterloop (Variables|Default|GetInWaterLoop)
          (:then
            (Audio|Components|Audio|Stop _inwaterloop))
          (:else
            _))
        (Audio|SpawnSoundAttached "/Game/Athena/Sounds/Vehicles/GolfCart/GolfCart_Movement_Impact_Water_Cue.GolfCart_Movement_Impact_Water_Cue" (Variables|Vehicle|GetSkeletalMesh))
        (bind _returnvalue_3 (Audio|SpawnSoundAttached "/Game/Athena/Sounds/Vehicles/GolfCart/GolfCart_Movement_InWater_Loop_Cue.GolfCart_Movement_InWater_Loop_Cue" (Variables|Vehicle|GetSkeletalMesh)))
        (Variables|Default|SetInWaterLoop _returnvalue_3)))))

(event EventOnDeathPlayEffects
  (params (Damage :type Float%20%28single-precision%29) (DamageTags :type Gameplay%20Tag%20Container%20Structure%20%28by%20ref%29) (Momentum :type Vector :default "0, 0, 0") (HitInfo :type Hit%20Result%20Structure%20%28by%20ref%29) (InstigatedBy :type Fort%20Pawn%20Object%20Reference) (DamageCauser :type Actor%20Object%20Reference) (EffectContext :type Gameplay%20Effect%20Context%20Handle%20Structure))
  (bind _returnvalue (Transformation|GetActorLocation))
  (bind (_returnvalue_location _returnvalue_rotation) (Math|Transform|BreakTransform (Transformation|GetSocketTransform (Variables|Vehicle|GetSkeletalMesh) "DriverSeat" "RTS_World")))
  ;[1 0.3615 0.07292] EVENT -> On Death Play Effects
  (bind _returnvalue_2 (Effects|SpawnEmitteratLocation "/Game/Athena/DrivableVehicles/FX/Octopus/P_Octopus_Vehicle_Destroyed.P_Octopus_Vehicle_Destroyed" _returnvalue_location _returnvalue_rotation "1.000000,1.000000,1.000000" true "AutoRelease"))
  (Effects|Components|ParticleSystem|SetVectorParameter _returnvalue_2 "Velocity" (Transformation|GetVelocity))
  (Camera|PlayWorldCameraShake "/Game/Effects/Fort_Effects/BluePrints/CameraShakes/B_Small_Vertical_Jolt_CameraShake.B_Small_Vertical_Jolt_CameraShake_C" _returnvalue 1.5e+02 7e+02)
  (Fortnite|PlayLocalForceFeedbackatLocationMulti "/Game/Weapons/ForceFeedback/Explosions/FF_Generic_Explosion_Near.FF_Generic_Explosion_Near" "/Game/Weapons/ForceFeedback/Explosions/FF_Generic_Explosion_Far.FF_Generic_Explosion_Far" _returnvalue 5e+02 1.2e+03 "")
  (Development|SetHiddeninGame (Variables|Vehicle|GetSkeletalMesh) true false)
  (Audio|PlaySoundatLocation "/Game/Athena/Sounds/Vehicles/Hamster/SC_Hamster_CollisionImpact_Destroy_Cue.SC_Hamster_CollisionImpact_Destroy_Cue" _returnvalue))

(event Custom|CheckWater
  (bind (_returnvalue_roll _returnvalue_pitch _returnvalue_yaw) (Math|Rotator|BreakRotator (Transformation|GetSocketRotation (Variables|Vehicle|GetSkeletalMesh) "DriverSeat")))
  ;[1 1 1] EVENT -> On Entered Water Volume -- FX for water ineraction
  (bind _inwaterfx (Variables|Default|GetInWaterFX)
    (:then
      (bind (_returnvalue_x _returnvalue_y _returnvalue_z) (Math|Vector|BreakVector (- (stale Vehicle|Water|EventOnEnteredWaterVolume WaterSurfacePoint) "0, 0,100.000000")))
      (Transformation|SetRelativeLocation _inwaterfx (Math|Vector|MakeVector 0.0 0.0 _returnvalue_z) false false)
      (Transformation|SetWorldRotation _inwaterfx (Math|Rotator|MakeRotator 0.0 0.0 _returnvalue_yaw) false false)
      (Effects|Components|ParticleSystem|SetVectorParameter (Variables|Default|GetInWaterFX) "Offset" (Math|Vector|Normalize (Math|Transform|InverseTransformDirection (Transformation|GetWorldTransform (Variables|Default|GetInWaterFX)) (Transformation|GetVelocity))))
      (bind _inwaterloop (Variables|Default|GetInWaterLoop)
        (:then
          (Audio|Components|Audio|SetVolumeMultiplier _inwaterloop (Math|Float|MapRangeClamped (Vehicle|GetForwardSpeedKmh) 0.0 2e+01 0.0 1.0) :!class /Script/Engine.AudioComponent))
        (:else)))
    (:else)))

(event Vehicle|Water|EventOnExitedWaterVolume
  (params (WaterSurfacePoint :type Vector%20%28by%20ref%29 :default "0, 0, 0"))
  ;[1 1 1] EVENT -> On Entered Water Volume -- FX for water ineraction
  (Networking|CanExecuteCosmeticEvents
    (:True
      (Utilities|Time|ClearandInvalidateTimerbyHandle (stale Vehicle|Water|EventOnEnteredWaterVolume _returnvalue))
      (bind _inwaterfx (Variables|Default|GetInWaterFX)
        (:then
          (Components|Activation|SetActive _inwaterfx false false)
          (bind _inwaterloop (Variables|Default|GetInWaterLoop)
            (:then
              (Audio|Components|Audio|FadeOut _inwaterloop 3.0 0.0 "Linear" :!class /Script/Engine.AudioComponent))
            (:else)))
        (:else)))))

(event EventOnHoldExitStarted
  (params (ExitingPawn :type Fort%20Pawn%20Object%20Reference) (ExitDuration :type Float%20%28single-precision%29))
  ;[1 1 1] EVENT -> OnHoldExitStarted/Stopped
  (Ability|SendGameplayEventtoActor ExitingPawn "(TagName=\"Athena.Vehicle.Biplane.OnHoldExitStarted\")" (Utilities|Struct|MakeGameplayEventData :ContextHandle "()" :InstigatorTags "(GameplayTags=)" :TargetTags "(GameplayTags=)" :EventMagnitude ExitDuration :TargetData "()")))

(event EventShowCooldownCue
  (params (Pawn :type Fort%20Player%20Pawn%20Object%20Reference) (Duration :type Float%20%28single-precision%29))
  ;[1 1 1] Cooldow Cue
  (Ability|SendGameplayEventtoActor Pawn "(TagName=\"GameplayEffect.Vehicle.Boost\")" (Utilities|Struct|MakeGameplayEventData :ContextHandle "()" :InstigatorTags "(GameplayTags=)" :TargetTags "(GameplayTags=)" :EventMagnitude Duration :TargetData "()")))

(event EventOnHoldExitStopped
  (params (ExitingPawn :type Fort%20Pawn%20Object%20Reference))
  ;[1 1 1] EVENT -> OnHoldExitStarted/Stopped
  (Ability|SendGameplayEventtoActor ExitingPawn "(TagName=\"Athena.Vehicle.Biplane.OnHoldExitStopped\")"))

(fn AddCooldownCueEvent
  (params (Player%20Pawn :type Fort%20Player%20Pawn%20Object%20Reference))
  (bind _returnvalue (Environment|ApplyGlobalEnvironmentGameplayEffecttoActor Player%20Pawn (LocalVariables|GetGE_BoostCooldown)))
  (Variables|Default|SetCooldownCueEvent _returnvalue))

(fn HandleOnHealthChanged
  (bind _damagestates (Variables|Default|GetDamageStates))
  (bind _returnvalue (Utilities|Array|LastIndex _damagestates))
  (bind _returnvalue_2 (- _returnvalue 1))
  (Networking|CanExecuteCosmeticEvents
    (:True
      (LocalVariables|SetHealthRatio (/ (Class|FortVehicleInterface|GetHealth) (Class|FortVehicleInterface|GetMaxHealth)))
      (Utilities|FlowControl|Sequence
        (:then_0
          (if (> (LocalVariables|GetHealthRatio) (Utilities|Array|Get(aref) (Variables|Default|GetDamageStates) 1))
            (LocalVariables|SetDesiredDamageStateIndex 0)
            (elif (<= (LocalVariables|GetHealthRatio) (Utilities|Array|Get(aref) (Variables|Default|GetDamageStates) _returnvalue_2))
              (LocalVariables|SetDesiredDamageStateIndex _returnvalue_2)
              (else
                (bind _index (Utilities|FlowControl|ForLoopwithBreak 0 (- (Utilities|Array|LastIndex (Variables|Default|GetDamageStates)) 1))
                  (:LoopBody
                    (if (and (> (Utilities|Array|Get(aref) (Variables|Default|GetDamageStates) _index) (LocalVariables|GetHealthRatio)) (<= (Utilities|Array|Get(aref) (Variables|Default|GetDamageStates) (+ _index 1)) (LocalVariables|GetHealthRatio)))
                      (LocalVariables|SetDesiredDamageStateIndex _index))))))))
        (:then_1
          (if (!= (LocalVariables|GetDesiredDamageStateIndex) (Variables|Default|GetDamageStateIndex))
            (Utilities|FlowControl|Sequence
              (:then_0
                (Rendering|Material|SetScalarParameterValueonMaterials (Variables|Vehicle|GetSkeletalMesh) "Health" (Utilities|Array|Get(aref) (Variables|Default|GetDamageScalarParameterValues) (LocalVariables|GetDesiredDamageStateIndex)))
                (Effects|Components|ParticleSystem|SetFloatParameter (Variables|Default|GetIdleFXArray) "Damage" (Math|Float|SelectFloat 1.0 0.0 (>= (LocalVariables|GetDesiredDamageStateIndex) 2)))
                (Effects|Components|ParticleSystem|SetFloatParameter (Variables|Default|GetIdleFXArray) "BadlyDamaged" (Math|Float|SelectFloat 1.0 0.0 (>= (LocalVariables|GetDesiredDamageStateIndex) 3))))
              (:then_1
                (if (> (LocalVariables|GetDesiredDamageStateIndex) (Variables|Default|GetDamageStateIndex))
                  (Effects|SpawnEmitterAttached :EmitterTemplate "/Game/Athena/DrivableVehicles/FX/Octopus/P_Octopus_Damage.P_Octopus_Damage" :AttachToComponent (Variables|Vehicle|GetSkeletalMesh) :AttachPointName "DriverSeat" :Scale "1.000000,1.000000,1.000000" :LocationType "SnapToTarget" :bAutoDestroy true :PoolingMethod "None" :bAutoActivate true)
                  (Audio|SpawnSoundAttached (Variables|Sounds|GetSoundOnDamageThreshold) (Variables|Vehicle|GetSkeletalMesh))))
              (:then_2
                (Variables|Default|SetDamageStateIndex (LocalVariables|GetDesiredDamageStateIndex))))))))))

(fn RemoveCooldownCueEvent
  (params (Player%20Pawn :type Fort%20Player%20Pawn%20Object%20Reference))
  (Ability|SendGameplayEventtoActor Player%20Pawn (LocalVariables|GetGE_BoostCooldown) (Utilities|Struct|MakeGameplayEventData :ContextHandle "()" :InstigatorTags "(GameplayTags=)" :TargetTags "(GameplayTags=)" :EventMagnitude 0.0 :TargetData "()"))
  (GameplayEffects|RemoveActiveGameplayEffect (Ability|GetAbilitySystemComponent Player%20Pawn) (Variables|Default|GetCooldownCueEvent) -1))

(fn ConstructionScript
  (Variables|Default|SetIdleFXArray (Utilities|Array|MakeArray (Variables|Default|GetP_Octopus_Coil_Idle_TopR) (Variables|Default|GetP_Octopus_Coil_Idle_LowR) (Variables|Default|GetP_Octopus_Coil_Idle_TopL) (Variables|Default|GetP_Octopus_Coil_Idle_LowL))))
