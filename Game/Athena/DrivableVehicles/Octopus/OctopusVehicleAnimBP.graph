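# Animation Blueprint for the Octopus drivable vehicle on the Hamster Ball skeleton, animating movement, boosting and grapple firing.
# A Locomotion state machine blends a cached directional blendspace, cardinal-direction pivots, boost states and grapple states with driver-gender variants.
# World-space bone overrides then rotate the shell, seat ring and towhook independently, and scale the plunger.
# The event graph zeroes plunger scale while the towhook is unholstered, restoring it once holstered with rope retracted.
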AnimBlueprint /Game/Athena/DrivableVehicles/Octopus/OctopusVehicleAnimBP
target_skeleton=/Game/Athena/DrivableVehicles/Hamster_Ball/meshes/Hamster_Ball_Veh_Skeleton
graphs [
 AnimGraph {nodes [
   AnimGraphNode_StateMachine_0 = Locomotion() {is=title AnimGraphNode_StateMachine}
   n1 = "Transform (Modify) Bone - Bone: ball_COM"!(ComponentPose=Local To Component(LocalPose=@AnimGraphNode_StateMachine_0), Alpha=1, Rotation=Get ShellRotation() -> (ShellRotation:Rotator)) -> (Pose) {rotationMode=BMM_Replace, rotationSpace=BCS_WorldSpace, alphaBoolBlend {blendOption=Linear}, boneToModify {boneName=ball_COM}}
   n2 = "Transform (Modify) Bone - Bone: inner_ring"!(ComponentPose=@n1, Alpha=1, Rotation=Get SeatRotation() -> (SeatRotation:Rotator)) -> (Pose) {rotationMode=BMM_Replace, rotationSpace=BCS_WorldSpace, alphaBoolBlend {blendOption=Linear}, boneToModify {boneName=inner_ring}}
   n3 = "Transform (Modify) Bone - Bone: plunger_radius"!(ComponentPose=@n2, Alpha=1, Rotation=Get TowhookRotation() -> (TowhookRotation:Rotator)) -> (Pose) {rotationMode=BMM_Replace, rotationSpace=BCS_WorldSpace, alphaBoolBlend {blendOption=Linear}, boneToModify {boneName=plunger_radius}}
   n4 = "Transform (Modify) Bone - Bone: plunger"!(ComponentPose=@n3, Alpha=1, Scale=Get PlungerScaleVector() -> (PlungerScaleVector:Vector)) -> (Pose) {rotationSpace=BCS_WorldSpace, scaleMode=BMM_Replace, alphaBoolBlend {blendOption=Linear}, boneToModify {boneName=plunger}}
   Output Pose(Result=Component To Local(ComponentPose=@n4)) {layerGroup=None}
   Save cached pose 'DirectionPose'(Pose=Blendspace Player 'OctopusVehicle_Direction_BS_Vehicle'(X=Get LeftRight() -> (LeftRight:float), Y=Get FwdBwd() -> (FwdBwd:float)) {blendSpace=/Game/Animation/Game/MainPlayer/Vehicles/OctopusVehicle/Vehicle/OctopusVehicle_Direction_BS_Vehicle}) {cacheName=DirectionPose}
  ]}
 Locomotion {
  entry=Default_InPlace
  path=AnimGraph.AnimGraphNode_StateMachine_0.Locomotion
  ids=id
  nodes [
   AnimStateNode_0 = Default_InPlace() {is=title AnimStateNode}
   AnimStateNode_1 = Boosting_Intro() {is=title AnimStateNode}
   AnimStateNode_2 = Boosting_Loop() {is=title AnimStateNode}
   AnimStateNode_3 = GrappleFireIntro() {is=title AnimStateNode}
   AnimStateNode_4 = GrappleFireLoop() {is=title AnimStateNode}
   AnimStateNode_5 = Default() {is=title AnimStateNode}
   AnimStateNode_6 = Pivot() {is=title AnimStateNode}
   AnimStateNode_7 = GrapplerFireFromBoosting() {is=title AnimStateNode}
   AnimStateTransitionNode_0 = AnimStateTransitionNode()
   AnimStateTransitionNode_1 = AnimStateTransitionNode() {bAutomaticRuleBasedOnSequencePlayerInState=true, crossfadeDuration=0.45}
   AnimStateTransitionNode_11 = AnimStateTransitionNode() {crossfadeDuration=0.15}
   AnimStateTransitionNode_12 = AnimStateTransitionNode() {bAutomaticRuleBasedOnSequencePlayerInState=true, crossfadeDuration=0.15}
   AnimStateTransitionNode_13 = AnimStateTransitionNode() {crossfadeDuration=0.067}
   AnimStateTransitionNode_14 = AnimStateTransitionNode()
   AnimStateTransitionNode_15 = AnimStateTransitionNode() {crossfadeDuration=0.15}
   AnimStateTransitionNode_16 = AnimStateTransitionNode() {crossfadeDuration=0.15}
   AnimStateTransitionNode_17 = AnimStateTransitionNode()
   AnimStateTransitionNode_18 = AnimStateTransitionNode() {crossfadeDuration=0.067}
   AnimStateTransitionNode_19 = AnimStateTransitionNode() {crossfadeDuration=0.15}
   AnimStateTransitionNode_2 = AnimStateTransitionNode()
   AnimStateTransitionNode_20 = AnimStateTransitionNode()
   AnimStateTransitionNode_3 = AnimStateTransitionNode() {crossfadeDuration=0}
   AnimStateTransitionNode_4 = AnimStateTransitionNode() {bAutomaticRuleBasedOnSequencePlayerInState=true, crossfadeDuration=0, priorityOrder=2}
   AnimStateTransitionNode_5 = AnimStateTransitionNode() {crossfadeDuration=0.45, priorityOrder=2}
   AnimStateTransitionNode_6 = AnimStateTransitionNode() {crossfadeDuration=0.1}
   AnimStateTransitionNode_7 = AnimStateTransitionNode() {crossfadeDuration=0.15}
   AnimStateTransitionNode_8 = AnimStateTransitionNode() {crossfadeDuration=0.15}
   AnimStateTransitionNode_9 = AnimStateTransitionNode() {bAutomaticRuleBasedOnSequencePlayerInState=true}
  ]
  links [
   {
    dst=Boosting_Loop
    rule=AnimGraph.AnimGraphNode_StateMachine_0.Locomotion.AnimStateTransitionNode_4.Transition
    src=Boosting_Intro
    via=AnimStateTransitionNode_4
   }
   {
    dst=GrapplerFireFromBoosting
    rule=AnimGraph.AnimGraphNode_StateMachine_0.Locomotion.AnimStateTransitionNode_7.Transition
    src=Boosting_Intro
    via=AnimStateTransitionNode_7
   }
   {
    dst=Default
    rule=AnimGraph.AnimGraphNode_StateMachine_0.Locomotion.AnimStateTransitionNode_5.Transition
    src=Boosting_Loop
    via=AnimStateTransitionNode_5
   }
   {
    dst=GrapplerFireFromBoosting
    rule=AnimGraph.AnimGraphNode_StateMachine_0.Locomotion.AnimStateTransitionNode_8.Transition
    src=Boosting_Loop
    via=AnimStateTransitionNode_8
   }
   {
    dst=Boosting_Intro
    rule=AnimGraph.AnimGraphNode_StateMachine_0.Locomotion.AnimStateTransitionNode_3.Transition
    src=Default
    via=AnimStateTransitionNode_3
   }
   {
    dst=Default_InPlace
    rule=AnimGraph.AnimGraphNode_StateMachine_0.Locomotion.AnimStateTransitionNode_17.Transition
    src=Default
    via=AnimStateTransitionNode_17
   }
   {
    dst=GrappleFireIntro
    rule=AnimGraph.AnimGraphNode_StateMachine_0.Locomotion.AnimStateTransitionNode_11.Transition
    src=Default
    via=AnimStateTransitionNode_11
   }
   {
    dst=GrappleFireLoop
    rule=AnimGraph.AnimGraphNode_StateMachine_0.Locomotion.AnimStateTransitionNode_19.Transition
    src=Default
    via=AnimStateTransitionNode_19
   }
   {
    dst=Pivot
    rule=AnimGraph.AnimGraphNode_StateMachine_0.Locomotion.AnimStateTransitionNode_0.Transition
    src=Default
    via=AnimStateTransitionNode_0
   }
   {
    dst=Default
    rule=AnimGraph.AnimGraphNode_StateMachine_0.Locomotion.AnimStateTransitionNode_14.Transition
    src=Default_InPlace
    via=AnimStateTransitionNode_14
   }
   {
    dst=GrappleFireIntro
    rule=AnimGraph.AnimGraphNode_StateMachine_0.Locomotion.AnimStateTransitionNode_15.Transition
    src=Default_InPlace
    via=AnimStateTransitionNode_15
   }
   {
    dst=GrappleFireLoop
    rule=AnimGraph.AnimGraphNode_StateMachine_0.Locomotion.AnimStateTransitionNode_16.Transition
    src=Default_InPlace
    via=AnimStateTransitionNode_16
   }
   {
    dst=Boosting_Intro
    rule=AnimGraph.AnimGraphNode_StateMachine_0.Locomotion.AnimStateTransitionNode_18.Transition
    src=GrappleFireIntro
    via=AnimStateTransitionNode_18
   }
   {
    dst=GrappleFireLoop
    rule=AnimGraph.AnimGraphNode_StateMachine_0.Locomotion.AnimStateTransitionNode_12.Transition
    src=GrappleFireIntro
    via=AnimStateTransitionNode_12
   }
   {
    dst=Boosting_Intro
    rule=AnimGraph.AnimGraphNode_StateMachine_0.Locomotion.AnimStateTransitionNode_13.Transition
    src=GrappleFireLoop
    via=AnimStateTransitionNode_13
   }
   {
    dst=Default
    rule=AnimGraph.AnimGraphNode_StateMachine_0.Locomotion.AnimStateTransitionNode_20.Transition
    src=GrappleFireLoop
    via=AnimStateTransitionNode_20
   }
   {
    dst=Boosting_Loop
    rule=AnimGraph.AnimGraphNode_StateMachine_0.Locomotion.AnimStateTransitionNode_9.Transition
    src=GrapplerFireFromBoosting
    via=AnimStateTransitionNode_9
   }
   {
    dst=Boosting_Intro
    rule=AnimGraph.AnimGraphNode_StateMachine_0.Locomotion.AnimStateTransitionNode_2.Transition
    src=Pivot
    via=AnimStateTransitionNode_2
   }
   {
    dst=Default
    rule=AnimGraph.AnimGraphNode_StateMachine_0.Locomotion.AnimStateTransitionNode_1.Transition
    src=Pivot
    via=AnimStateTransitionNode_1
   }
   {
    dst=GrappleFireIntro
    rule=AnimGraph.AnimGraphNode_StateMachine_0.Locomotion.AnimStateTransitionNode_6.Transition
    src=Pivot
    via=AnimStateTransitionNode_6
   }
  ]
 }
 Default_InPlace {
  path=AnimGraph.AnimGraphNode_StateMachine_0.Locomotion.AnimStateNode_0.Default_InPlace
  nodes [Output Animation Pose(Result=Use cached pose 'DirectionPose'())]
 }
 Boosting_Intro {
  path=AnimGraph.AnimGraphNode_StateMachine_0.Locomotion.AnimStateNode_1.Boosting_Intro
  nodes [
   n1 = Sequence Player 'OctopusVehicle_Boosting_Intro_Vehicle'() {bLoopAnimation=false, sequence=/Game/Animation/Game/MainPlayer/Vehicles/OctopusVehicle/Vehicle/OctopusVehicle_Boosting_Intro_Vehicle}
   Output Animation Pose(Result=@n1)
  ]
 }
 Boosting_Loop {
  path=AnimGraph.AnimGraphNode_StateMachine_0.Locomotion.AnimStateNode_2.Boosting_Loop
  nodes [Output Animation Pose(Result=Sequence Player 'OctopusVehicle_Boosting_Pose_Vehicle'() {sequence=/Game/Animation/Game/MainPlayer/Vehicles/OctopusVehicle/Vehicle/OctopusVehicle_Boosting_Pose_Vehicle})]
 }
 GrappleFireIntro {
  path=AnimGraph.AnimGraphNode_StateMachine_0.Locomotion.AnimStateNode_3.GrappleFireIntro
  nodes [
   n1 = Sequence Player 'OctopusVehicle_GrappleFire_Intro_CMF_Vehicle'() {sequence=/Game/Animation/Game/MainPlayer/Vehicles/OctopusVehicle/Vehicle/OctopusVehicle_GrappleFire_Intro_CMF_Vehicle, startPosition=0.2}
   n2 = Sequence Player 'OctopusVehicle_GrappleFire_Intro_Vehicle'() {
    bLoopAnimation=false
    sequence=/Game/Animation/Game/MainPlayer/Vehicles/OctopusVehicle/Vehicle/OctopusVehicle_GrappleFire_Intro_Vehicle
    startPosition=0.2
   }
   Output Animation Pose(Result=Blend Poses by bool(BlendPose_0=@n1, BlendPose_1=@n2, bActiveValue=Get bIsDriverFemale() -> (bIsDriverFemale:bool)) {blendType=Linear})
  ]
 }
 GrappleFireLoop {
  path=AnimGraph.AnimGraphNode_StateMachine_0.Locomotion.AnimStateNode_4.GrappleFireLoop
  nodes [Output Animation Pose(Result=Blend Poses by bool(
     BlendPose_0=Sequence Player 'OctopusVehicle_GrappleFire_Pose_CMF_Vehicle'() {sequence=/Game/Animation/Game/MainPlayer/Vehicles/OctopusVehicle/Vehicle/OctopusVehicle_GrappleFire_Pose_CMF_Vehicle}
     BlendPose_1=Sequence Player 'OctopusVehicle_GrappleFire_Pose_Vehicle'() {sequence=/Game/Animation/Game/MainPlayer/Vehicles/OctopusVehicle/Vehicle/OctopusVehicle_GrappleFire_Pose_Vehicle}
     bActiveValue=Get bIsDriverFemale()
    ) {blendType=Linear})]
 }
 Default {
  path=AnimGraph.AnimGraphNode_StateMachine_0.Locomotion.AnimStateNode_5.Default
  nodes [Output Animation Pose(Result=Use cached pose 'DirectionPose'())]
 }
 Pivot {
  path=AnimGraph.AnimGraphNode_StateMachine_0.Locomotion.AnimStateNode_6.Pivot
  nodes [
   n1 = "Blend Poses (EFortCardinalDirection)"!(
    BlendPose_0:Pose Link Structure=Sequence Player 'OctopusVehicle_Pivot_NtoS_Vehicle'() {bLoopAnimation=false, sequence=/Game/Animation/Game/MainPlayer/Vehicles/OctopusVehicle/Vehicle/OctopusVehicle_Pivot_NtoS_Vehicle}
    BlendPose_1:Pose Link Structure=Sequence Player 'OctopusVehicle_Pivot_StoN_Vehicle'() {bLoopAnimation=false, sequence=/Game/Animation/Game/MainPlayer/Vehicles/OctopusVehicle/Vehicle/OctopusVehicle_Pivot_StoN_Vehicle}
    BlendPose_2:Pose Link Structure=Sequence Player 'OctopusVehicle_Pivot_WtoE_Vehicle'() {bLoopAnimation=false, sequence=/Game/Animation/Game/MainPlayer/Vehicles/OctopusVehicle/Vehicle/OctopusVehicle_Pivot_WtoE_Vehicle}
    BlendPose_3:Pose Link Structure=Sequence Player 'OctopusVehicle_Pivot_EtoW_Vehicle'() {bLoopAnimation=false, sequence=/Game/Animation/Game/MainPlayer/Vehicles/OctopusVehicle/Vehicle/OctopusVehicle_Pivot_EtoW_Vehicle}
    BlendTime_0:float=0.3
    BlendTime_1:float=0.3
    BlendTime_2:float=0.3
    BlendTime_3:float=0.3
    ActiveEnumValue:Byte=Get PivotDir() -> (PivotDir:EFortCardinalDirection Enum=North)
   ) -> (Pose:Pose Link Structure) {
    is=title AnimGraphNode_BlendListByEnum
    blendType=Linear
    blendPose [{linkId=-1, sourceLinkId=-1}, {linkId=-1, sourceLinkId=-1}, {linkId=-1, sourceLinkId=-1}, {linkId=-1, sourceLinkId=-1}]
    blendTime [0.1,0.1,0.1,0.1]
   }
   Output Animation Pose(Result=@n1)
  ]
 }
 GrapplerFireFromBoosting {
  path=AnimGraph.AnimGraphNode_StateMachine_0.Locomotion.AnimStateNode_7.GrapplerFireFromBoosting
  nodes [
   n1 = Sequence Player 'OctopusVehicle_GrappleFire_Boosting_Vehicle'() {
    bLoopAnimation=false
    playRate=1.1
    sequence=/Game/Animation/Game/MainPlayer/Vehicles/OctopusVehicle/Vehicle/OctopusVehicle_GrappleFire_Boosting_Vehicle
   }
   Output Animation Pose(Result=@n1)
  ]
 }
 Transition {
  path=AnimGraph.AnimGraphNode_StateMachine_0.Locomotion.AnimStateTransitionNode_0.Transition
  nodes [Result(bCanEnterTransition=Get bShouldPlayPivotTransition() -> (bShouldPlayPivotTransition:bool))]
 }
 Transition {path=AnimGraph.AnimGraphNode_StateMachine_0.Locomotion.AnimStateTransitionNode_1.Transition, nodes [Result()]}
 Transition {
  path=AnimGraph.AnimGraphNode_StateMachine_0.Locomotion.AnimStateTransitionNode_11.Transition
  nodes [Result(bCanEnterTransition=Get bShouldPlayGrappleFire() -> (bShouldPlayGrappleFire:bool))]
 }
 Transition {path=AnimGraph.AnimGraphNode_StateMachine_0.Locomotion.AnimStateTransitionNode_12.Transition, nodes [Result()]}
 Transition {
  path=AnimGraph.AnimGraphNode_StateMachine_0.Locomotion.AnimStateTransitionNode_13.Transition
  nodes [Result(bCanEnterTransition=Get bIsBoosting() -> (bIsBoosting:bool))]
 }
 Transition {
  path=AnimGraph.AnimGraphNode_StateMachine_0.Locomotion.AnimStateTransitionNode_14.Transition
  nodes [Result(bCanEnterTransition=NOT Boolean(A=Get bIsLowerVelocity() -> (bIsLowerVelocity:bool)))]
 }
 Transition {
  path=AnimGraph.AnimGraphNode_StateMachine_0.Locomotion.AnimStateTransitionNode_15.Transition
  nodes [Result(bCanEnterTransition=Get bShouldPlayGrappleFire())]
 }
 Transition {
  path=AnimGraph.AnimGraphNode_StateMachine_0.Locomotion.AnimStateTransitionNode_16.Transition
  nodes [Result(bCanEnterTransition=Get bIsTowhookAttached() -> (bIsTowhookAttached:bool))]
 }
 Transition {
  path=AnimGraph.AnimGraphNode_StateMachine_0.Locomotion.AnimStateTransitionNode_17.Transition
  nodes [Result(bCanEnterTransition=Get bIsLowerVelocity())]
 }
 Transition {
  path=AnimGraph.AnimGraphNode_StateMachine_0.Locomotion.AnimStateTransitionNode_18.Transition
  nodes [Result(bCanEnterTransition=Get bIsBoosting())]
 }
 Transition {
  path=AnimGraph.AnimGraphNode_StateMachine_0.Locomotion.AnimStateTransitionNode_19.Transition
  nodes [Result(bCanEnterTransition=Get bIsTowhookAttached())]
 }
 Transition {
  path=AnimGraph.AnimGraphNode_StateMachine_0.Locomotion.AnimStateTransitionNode_2.Transition
  nodes [Result(bCanEnterTransition=Get bIsBoosting())]
 }
 Transition {
  path=AnimGraph.AnimGraphNode_StateMachine_0.Locomotion.AnimStateTransitionNode_20.Transition
  nodes [Result(bCanEnterTransition=Get bIsTowhookHolstered() -> (bIsTowhookHolstered:bool))]
 }
 Transition {
  path=AnimGraph.AnimGraphNode_StateMachine_0.Locomotion.AnimStateTransitionNode_3.Transition
  nodes [Result(bCanEnterTransition=Get bIsBoosting())]
 }
 Transition {path=AnimGraph.AnimGraphNode_StateMachine_0.Locomotion.AnimStateTransitionNode_4.Transition, nodes [Result()]}
 Transition {
  path=AnimGraph.AnimGraphNode_StateMachine_0.Locomotion.AnimStateTransitionNode_5.Transition
  nodes [Result(bCanEnterTransition=NOT Boolean(A=Get bIsBoosting()))]
 }
 Transition {
  path=AnimGraph.AnimGraphNode_StateMachine_0.Locomotion.AnimStateTransitionNode_6.Transition
  nodes [Result(bCanEnterTransition=Get bShouldPlayGrappleFire())]
 }
 Transition {
  path=AnimGraph.AnimGraphNode_StateMachine_0.Locomotion.AnimStateTransitionNode_7.Transition
  nodes [Result(bCanEnterTransition=Get bShouldPlayGrappleFire())]
 }
 Transition {
  path=AnimGraph.AnimGraphNode_StateMachine_0.Locomotion.AnimStateTransitionNode_8.Transition
  nodes [Result(bCanEnterTransition=Get bShouldPlayGrappleFire())]
 }
 Transition {path=AnimGraph.AnimGraphNode_StateMachine_0.Locomotion.AnimStateTransitionNode_9.Transition, nodes [Result()]}
 EventGraph {nodes [
   Event BlueprintUpdateAnimation()
   > n1 = Branch(Condition=Get bIsTowhookHolstered())
   > n2 = Cast To OctopusVehicle!(execute, Object=Get OctopusVehicle() -> (OctopusVehicle:Fort Octopus Vehicle Object Reference)) -> (then, CastFailed, AsOctopus Vehicle:Octopus Vehicle Object Reference) {is=title K2Node_DynamicCast}
   > Branch(Condition=Get bHasRopeRetracted!(self:Octopus Vehicle Object Reference=@n2) -> (bHasRopeRetracted:bool))
   > Set PlungerScaleVector(PlungerScaleVector:Vector="1,1,1") -> (Output_Get:Vector)
   Reroute Node(InputPin=@n2.CastFailed)
   Set PlungerScaleVector(execute=@n1.else, PlungerScaleVector="0, 0, 0")
  ]}
]
variables [PlungerScaleVector {type=Vector}]
