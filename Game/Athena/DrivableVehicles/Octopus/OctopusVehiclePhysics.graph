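# Physics asset for the Octopus drivable vehicle in Athena.
# It defines a single body on the root bone, made of one sphere raised above the bone origin.
# There are no constraints, so the vehicle simulates as one rigid ball with no articulated parts.
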PhysicsAsset /Game/Athena/DrivableVehicles/Octopus/OctopusVehiclePhysics
bodies {root [{shapeType=Sphere, radius=75, center=<0.2733 0 76.4117>}]}
constraints []
